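; Looping gameplay cue for the HighTower Soy board item that makes the board's loop audio react to the player's speed.
; On loop start it stores the target Fort player pawn and an audio component, then starts a repeating timer.
; Each timer tick maps movement velocity magnitude to the audio pitch multiplier.
; An FXTick event that fades in and lengthens a Niagara trail is marked disabled and is not scheduled here.
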(event EventOnLoopingStart(Cascade)
  (params (MyTarget :type Actor%20Object%20Reference) (Parameters :type Gameplay%20Cue%20Parameters%20Structure%20%28by%20ref%29) (ParticleComponents :type Array%20of%20Cascade%20Particle%20System%20Component%20Object%20References) (AudioComponents :type Array%20of%20Audio%20Component%20Object%20References))
  ;[1 1 1] Setup
  (bind _asfort_player_pawn (Utilities|Casting|CastToFortPlayerPawn MyTarget))
  (Variables|Default|SetTargetPlayerPawn _asfort_player_pawn)
  (Variables|Default|SetLoopingAudio (Utilities|Array|Get(acopy) AudioComponents))
  (Utilities|Time|SetTimerbyEvent (EventDispatchers|CreateEvent :!event Custom|AudioTick) 0.1 true))

(event Custom|AudioTick
  ;[1 1 1] Audio
  (bind _targetplayerpawn (Variables|Default|GetTargetPlayerPawn)
    (:then
      (Audio|Components|Audio|SetPitchMultiplier (Variables|Default|GetLoopingAudio) (Math|Float|MapRangeClamped (Math|Vector|VectorLength (Variables|/Script/Engine.MovementComponent|GetVelocity (Variables|/Script/Engine.Character|GetCharacterMovement _targetplayerpawn))) 0.0 1e+03 1.0 1.5)))
    (:else)))

(event Custom|FXTick
  ;[1 1 1] VFX Trail - DISABLED
  (bind _output_get (Variables|Default|SetPowerAge (- (Utilities|Time|GetGameTimeinSeconds 0) (Variables|Default|GetPowerStartTime))))
  ;;[0.3853 1 0.9901] Scales trail opacity over first second
  (Niagara|SetNiagaraVariableByString(Float) (Variables|Default|GetLoopingNiagaraSys) "User.BPOpacityMult" (Math|Float|Clamp(Float) _output_get))
  ;;[0.4208 1 0.4951] Scales trail mesh over first half second
  (Niagara|SetNiagaraVariableByString(Float) (Variables|Default|GetLoopingNiagaraSys) "User.TrailLengthScale" (Math|Float|Clamp(Float) (* (Variables|Default|GetPowerAge) 2.0))))

(fn ConstructionScript)
